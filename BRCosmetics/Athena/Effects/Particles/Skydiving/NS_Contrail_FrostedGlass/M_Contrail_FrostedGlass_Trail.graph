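# Translucent unlit Niagara ribbon material for the trail of the Frosted Glass skydiving contrail.
# A snowflake texture, panned and tiled along U by dynamic parameters, offsets a lookup into a thin-film rainbow texture, which is hue-shifted, boosted and added to particle color for emissive.
# Opacity erodes edge and length gradients through the snowflake red channel, applies an animated green-channel dissolve, then multiplies by particle alpha.
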Material /BRCosmetics/Athena/Effects/Particles/Skydiving/NS_Contrail_FrostedGlass/M_Contrail_FrostedGlass_Trail
graph {
 properties {
  bAutomaticallySetUsageInEditor=false
  bUsedWithNiagaraRibbons=true
  blendMode=BLEND_Translucent
  shadingModel=MSM_Unlit
  usedShadingModels=Unlit
 }
 nodes [
  n1 = ParticleColor()
  MaterialFunctionCall(Emissive=Add(A=NamedRerouteUsage(), B=@n1.RGB)) {materialFunction=&MF_FX_EmissiveAdjustments}
  Multiply(A=@n1.A, B=NamedRerouteUsage())
  n2 = ScalarParameter() {defaultValue=0.2, parameterName=HueShift}
  n3 = TextureCoordinate() {coordinateIndex=1}
  n4 = DynamicParameter() {paramNames [FakeSpecLevel_Mult, Dissolve_Anim, Pan_U, Tile_U]}
  n5 = AppendVector(A=@n4.Tile_U, B=Constant() {r=1})
  n6 = Constant()
  n7 = TextureSample(UVs=Panner(Coordinate=Multiply(A=@n3, B=@n5), Speed=AppendVector(A=@n4.Pan_U, B=@n6))) {samplerType=SAMPLERTYPE_LinearColor, texture=/Game/Effects/Fort_Effects/Textures/Snowflake/T_Snowflakes_Tile01}
  n8 = TextureSample(UVs=Sine(None=Add(A=@n2, B=@n7.R)) {period=2}) {
   samplerSource=SSM_Wrap_WorldGroupSettings
   texture=/BRCosmetics/Weapons/FORT_Melee/Pickaxe_RelayStick/FX/T_Pickaxe_RelayStick_ThinFilmRainbow
  }
  n9 = MaterialFunctionCall(Hue Shift Percentage=ScalarParameter() {defaultValue=0.55, parameterName=HueShift_Percentage}, Texture=@n8.RGB) {materialFunction=&HueShift}
  n10 = ScalarParameter() {defaultValue=3, parameterName=HueShift_Pow}
  NamedRerouteDeclaration(None=Multiply(A=Power(Base=@n9.Result, Exp=@n10) {constExponent=8}, B=ScalarParameter() {defaultValue=3, parameterName=HueShift_Mult}) {constB=3}) {name=Emissive Color, nodeColor=<rgba 0.3828 1 0 1>}
  n11 = Sine(None=ComponentMask(None=TextureCoordinate()) {g=true}) {period=2}
  n12 = ComponentMask(None=TextureCoordinate()) {r=true}
  n13 = OneMinus(None=@n12)
  n14 = ScalarParameter() {defaultValue=1.3, parameterName=Trail_Grad_Pow}
  n15 = Saturate(None=Multiply(A=Power(Base=@n13, Exp=@n14) {constExponent=1.3}, B=ScalarParameter() {defaultValue=1, parameterName=Trail_Grad_Mult}))
  n16 = Multiply(A=Power(Base=@n12) {constExponent=1}) {constB=8}
  n17 = SmoothStep(Min=OneMinus(None=@n7.R), Value=Multiply(A=Multiply(A=@n11, B=@n15), B=Saturate(None=@n16)))
  n18 = ScalarParameter() {defaultValue=1, parameterName=DissoveGrad_Hardness}
  n19 = Multiply(
   A=SmoothStep(Min=OneMinus(None=@n7.G), Value=DynamicParameter() {paramNames [FakeSpecLevel_Mult, Dissolve_Anim, Pan_U, Tile_U]}.Dissolve_Anim)
   B=Saturate(None=Divide(A=@n17, B=@n18))
  )
  n20 = ScalarParameter() {defaultValue=1, parameterName=DissolveAnim_Hardness}
  NamedRerouteDeclaration(None=Saturate(None=Divide(A=@n19, B=@n20))) {name=Opacity, nodeColor=<rgba 1 0 0.7031 1>}
 ]
}
