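# Frosted-glass skydiving contrail Niagara system that trails icy effects from the character's skeletal mesh sockets.
# Five emitters combine hand glow sprites with lights, ice fog sprites, drifting snowflake meshes, and two ribbon trails (frosty cubemap and ice fog).
# FN_Contrail_Setup's Dive Value blends offsets, velocities, widths and colors.
# Time-of-day curves modulate alpha on most layers, while front-end modifiers tone down several for menu display.
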NiagaraSystem /BRCosmetics/Athena/Effects/Particles/Skydiving/NS_Contrail_FrostedGlass/NS_Contrail_FrostedGlass
dependencies {
 usedDataInterfaces [&NiagaraDataInterfaceArrayFloat, &NiagaraDataInterfaceCurve, &NiagaraDataInterfaceColorCurve]
 usedDynamicInputs [&SimulationPosition]
 usedModules [
  &self:SystemSpawnScript.SystemScriptSource.SystemScriptGraph.NiagaraNodeAssignment_1.SetVariables_#1
  &self:SystemSpawnScript.SystemScriptSource.SystemScriptGraph.NiagaraNodeAssignment_0.SetVariables_#2
  &SystemState
  /Game/Effects/Niagara/Modules/Contrails/FN_Contrail_Setup
  &FN_Integrator
 ]
}
emitters {
 Hand_Glow {
  emitterUpdateScript [
   &EmitterState(Life Cycle Mode=::NewEnumerator1, Loop Behavior=::NewEnumerator1, Loop Duration Mode=::NewEnumerator1, UseLoopDelay=false)
   &SpawnBurst_Instantaneous(Spawn Count=2, Use Spawn Probability=false)
  ]
  particleSpawnScript [
   &FN_SetSpawnGroup(Number of Spawn Groups=2)
   &InitializeParticle(
    Color=<rgba 0 0.85 1.5 0.6>
    Lifetime=2
    Mass Mode=::NewEnumerator1
    Sprite Size Mode=::NewEnumerator3
    Uniform Sprite Size=120
    UsePositionOffset=false
    Write Material Random=true
   ) {module=InitializeParticle001}
  ]
  particleUpdateScript [
   &ParticleState(Kill Particles When Lifetime Has Elapsed=false)
   &ScaleSpriteSize(
    Uniform Curve Scale=&Sine(Normalized Angle=&Add_Float(A=$Emitter.Age, B=0), Period=0.8, Scale=0.1, Bias=0.8)
    Uniform Curve Sprite Scale=DataInterface() {propertyValues {Curve {keys [{tangentMode=RCTM_User, time=0, value=1}, {tangentMode=RCTM_User, time=1, value=1}]}}}
   )
   &SkeletalMeshLocation(
    Bone / Socket Index=$Particles.SpawnGroup
    Bone Sampling Mode=::NewEnumerator4
    Inherit Velocity=false
    Interpolate Mesh Sampling=false
    MaskSpawnGroup=false
    Sampled Position Offset=<0 0 7>
    Skeletal Mesh=$System.Skeletal Mesh
   )
   &CameraOffset()
   /SetVariables(Particles.Color_Decal:LinearColor=&MakeLinearColor(
     R:NiagaraFloat=0
     G:NiagaraFloat=0.3
     B:NiagaraFloat=1
     A:NiagaraFloat=&FloatFromCurve(FloatCurve=DataInterface() {propertyValues {Curve {keys [<at 0.0001 0.3382>,<at 0.1393 0.6251>,<at 0.3202 0.5233>,<at 0.4613 0.85>,<at 0.6241 0.3845>,<at 0.7724 0.6066>,<at 1.0006 0.3567>]}}})
    )) {isSetParameters=true}
   &ScaleColor(
    Scale Alpha=&FloatFromCurve(
     FloatCurve=DataInterface() {propertyValues {Curve {keys [<at 0 0.7 RCIM_Cubic>,<at 0.5 1 RCIM_Cubic>,<at 1 0.7 RCIM_Cubic>]}}}
     CurveIndex=Linked() {linkedVariable {name=NPC.FortniteNPC.TimeOfDay_0-1, type=NiagaraFloat}}
    )
    ScaleA=true
    ScaleRGB=true
   )
   &FN_Light_v2(Light Exponent=2, Light Radius=100, LightEnabled=true, Scale Color=2)
  ]
  rendererClasses [&NiagaraSpriteRendererProperties, &NiagaraLightRendererProperties]
  renderers [
   {rendererClass=&NiagaraSpriteRendererProperties}
   {rendererClass=&NiagaraLightRendererProperties, rendererIndex=1}
   {rendererClass=&NiagaraSpriteRendererProperties, rendererIndex=2}
  ]
 }
 Hand_IceFog {
  emitterUpdateScript [
   &EmitterState()
   &SpawnRate(SpawnRate=30, Use Spawn Probability=false)
  ]
  particleSpawnScript [
   &FN_SetSpawnGroup(Number of Spawn Groups:NiagaraInt32=&MakeIntFromFloat(
     Float=&ClampFloat(Float=&Lerp_Float(A=4, B=-2000, Alpha=$System.FN_Contrail_Setup.Dive Value), Min=4, Max=4)
     Conversion Mode=::NewEnumerator1
    ))
   &InitializeParticle(
    Sprite Rotation Mode=::NewEnumerator2
    Sprite Size Mode=::NewEnumerator3
    Sprite UV Mode=::NewEnumerator3
    Uniform Sprite Size=&Multiply_Float(
     A=&RandomRangeFloat(Minimum=40, Maximum=50)
     B=&self:SpawnGrp_0_1_Mask(Masked Float:NiagaraFloat=1, Passthrough Float:NiagaraFloat=1.2)
    )
    UsePositionOffset=false
    Write Material Random=true
   )
   &CameraOffset(Camera Offset Amount=&RandomRangeFloat(Minimum=0, Maximum=30))
  ]
  particleUpdateScript [
   &ParticleState()
   &ScaleColor(
    Scale Alpha=&FloatFromCurve(
     FloatCurve=DataInterface() {propertyValues {Curve {keys [<at 0 0 RCIM_Cubic>, {interpMode=RCIM_Cubic, tangentMode=RCTM_Break, time=0.2, value=1}, <at 1 0>]}}}
     Scale Curve=&Multiply_Float(
      A=&FN_FrontEndFloatModifier(Gameplay=1, Front End=0.3)
      B=&self:SpawnGrp_0_1_Mask(
       Masked Float:NiagaraFloat=&RandomRangeFloat(Minimum=0.2, Maximum=0.4)
       Passthrough Float:NiagaraFloat=1
      )
     )
    )
    Scale RGB=&MakeVectorFromLinearColorRGB(Color=&ColorFromCurve(ColorCurve=DataInterface() {propertyValues {
        AlphaCurve {keys []}
        BlueCurve {keys [<at 0.0767 3>,<at 0.3206 2>,<at 0.6485 1>]}
        GreenCurve {keys [<at 0.0767 3>,<at 0.3206 1.4984>,<at 0.6485 0.2188>]}
        RedCurve {keys [<at 0.0767 3>,<at 0.3206 0.1191>,<at 0.6485 0.2145>]}
       }}))
    ScaleA=true
    ScaleRGB=true
   )
   &ScaleColor(
    Scale Alpha=&FloatFromCurve(
     FloatCurve=DataInterface() {propertyValues {Curve {keys [
         {interpMode=RCIM_Cubic, leaveTangentWeight=0.0833, tangentMode=RCTM_User, time=0, value=0.2}
         {
          arriveTangent=0.7559
          arriveTangentWeight=0.1572
          interpMode=RCIM_Cubic
          leaveTangent=0.7559
          leaveTangentWeight=0.1572
          tangentMode=RCTM_User
          time=0.25
          value=0.25
         }
         {arriveTangentWeight=0.1681, interpMode=RCIM_Cubic, leaveTangentWeight=0.0841, tangentMode=RCTM_User, time=0.5, value=1}
         {
          arriveTangent=-0.6245
          arriveTangentWeight=0.1572
          interpMode=RCIM_Cubic
          leaveTangent=-0.6245
          leaveTangentWeight=0.1572
          tangentMode=RCTM_User
          time=0.75
          value=0.25
         }
         {arriveTangentWeight=0.0833, interpMode=RCIM_Cubic, tangentMode=RCTM_User, time=1, value=0.2}
        ]}}}
     CurveIndex=Linked() {linkedVariable {name=NPC.FortniteNPC.TimeOfDay_0-1, type=NiagaraFloat}}
    )
    ScaleA=true
    ScaleRGB=false
   ) {module=ScaleColor002}
   &ScaleSpriteSize(Uniform Curve Sprite Scale=DataInterface() {propertyValues {Curve {keys [
        {arriveTangent=3.5003, interpMode=RCIM_Cubic, leaveTangent=3.5003, tangentMode=RCTM_User, time=0, value=0.5}
        {arriveTangent=0.5833, interpMode=RCIM_Cubic, leaveTangent=0.5833, time=0.18, value=0.9}
        <at 1.2 1.2 RCIM_Cubic>
       ]}}})
   &SkeletalMeshLocation(
    Bone / Socket Index=$Particles.SpawnGroup
    Bone Sampling Mode=::NewEnumerator4
    MaskSpawnGroup=false
    Orientation Sampling=::NewEnumerator1
    Sampled Position Offset=&MakeVector(
     X=&Lerp_Float(
      A=0
      B=&MaskFloatBySpawnGroup(
       Masked Float=&FloatFromCurve(FloatCurve=DataInterface() {propertyValues {Curve {keys [<at 0 0>,<at 1 25>]}}})
       Passthrough Float=&FloatFromCurve(FloatCurve=DataInterface() {propertyValues {Curve {keys [<at 0 0>,<at 1 -25>]}}})
      )
      Alpha=$System.FN_Contrail_Setup.Dive Value
     )
     Y=&Lerp_Float(
      A=0
      B=&FloatFromCurve(
       FloatCurve=DataInterface() {propertyValues {Curve {keys [<at 0 0>,<at 1 -100>]}}}
       Scale Curve=&FN_FrontEndFloatModifier(Gameplay=1, Front End=0.3)
      )
      Alpha=$System.FN_Contrail_Setup.Dive Value
     )
     Z=&FloatFromCurve(
      FloatCurve=DataInterface() {propertyValues {Curve {keys [<at 0 0>,<at 1 80>]}}}
      Scale Curve=&self:SpawnGrp_0_1_Mask(Masked Float:NiagaraFloat=0.5, Passthrough Float:NiagaraFloat=1)
     )
    )
    Sampled Position Offset Coordinate Space:ENiagaraSkelMeshTransforms=::NewEnumerator2
    Skeletal Mesh=$System.Skeletal Mesh
    Use Sampled Mesh Attribute as Ribbon ID=false
   )
   &DynamicMaterialParameters(
    Index 0 Param 1=&FloatFromCurve(FloatCurve=DataInterface() {propertyValues {Curve {keys [<at 0 0 RCIM_Cubic>,<at 1.0011 0.2>]}}})
    Index 1 Param 1=&FloatFromCurve(
     FloatCurve=DataInterface() {propertyValues {Curve {keys [<at 0 0>,<at 0.25 1>]}}}
     Scale Curve=&RandomRangeFloat(Minimum=2, Maximum=1)
    )
    Index 1 Param 2=&FloatFromCurve(
     FloatCurve=DataInterface() {propertyValues {Curve {keys [
         {interpMode=RCIM_Cubic, leaveTangent=3.7422, tangentMode=RCTM_Break, time=0, value=0.2}
         {interpMode=RCIM_Cubic, leaveTangent=0.9066, time=1, value=2}
        ]}}}
     Scale Curve=&RandomRangeFloat(Minimum=0.5, Maximum=1)
    )
    Index 1 Param 3=&FloatFromCurve(FloatCurve=DataInterface() {propertyValues {Curve {keys [
         {interpMode=RCIM_Cubic, leaveTangent=2.8445, tangentMode=RCTM_Break, time=0, value=0.2}
         {interpMode=RCIM_Cubic, leaveTangent=63.2978, time=0.25, value=1}
        ]}}})
    Index 1 Param 4=&FloatFromCurve(
     FloatCurve=DataInterface() {propertyValues {Curve {keys [<at 0 1>,<at 1 0.8>]}}}
     Scale Curve=&RandomRangeFloat(Minimum=0.5, Maximum=1)
    )
    Index 2 Param 1=&FloatFromCurve(FloatCurve=DataInterface() {propertyValues {Curve {keys [{interpMode=RCIM_Cubic, leaveTangent=3.2309, tangentMode=RCTM_Break, time=0.1, value=0}, <at 1 1>]}}})
    Index 2 Param 2=$Particles.MaterialRandom
    Index 2 Param 3=&FloatFromCurve(FloatCurve=DataInterface() {propertyValues {Curve {keys [
         {interpMode=RCIM_Cubic, tangentMode=RCTM_Break, time=0, value=1}
         {interpMode=RCIM_Cubic, leaveTangent=11.1233, tangentMode=RCTM_Break, time=1, value=2}
        ]}}})
    Index 2 Param 4=0.75
    Module.Index0Param0Write=true
    Module.Index0Param1Write=true
    Module.Index0Param2Write=true
    Module.Index0Param3Write=true
    Module.Index1Param0Write=true
    Module.Index1Param1Write=true
    Module.Index1Param2Write=true
    Module.Index1Param3Write=true
    Module.Index2Param0Write=true
    Module.Index2Param1Write=true
    Module.Index2Param2Write=true
    Module.Index2Param3Write=true
    Write Parameter Index 1=true
    Write Parameter Index 2=true
   )
  ]
  rendererClasses [&NiagaraSpriteRendererProperties]
  renderers [{rendererClass=&NiagaraSpriteRendererProperties}]
 }
 Hand_SnowFlake {
  emitterUpdateScript [
   &EmitterState()
   &SpawnRate(SpawnGroup=1, SpawnRate=50, Use Spawn Probability=false)
  ]
  particleSpawnScript [
   &PartitionParticles()
   &InitializeParticle(
    Lifetime Max=0.65
    Lifetime Min=0.3
    Lifetime Mode=::NewEnumerator1
    Mesh Scale=&Multiply_VectorByFloat(Vector=<1 1 1>, Float=&RandomRangeFloat(Minimum=0.2, Maximum=0.4))
    Mesh Scale Mode=::NewEnumerator4
    UsePositionOffset=false
    Write Material Random=true
   ) {module=InitializeParticle001}
   &InitialMeshOrientation(Use Rotation Vector=false)
   &AddVelocity(Velocity=&RandomRangeVector(Minimum=<0 0 550>, Maximum=<0 0 600>))
   &AddVelocity(Velocity=&MakeVector(
     X=&SelectFloatFromArray(
      Direct Array Index=$Particles.PartitionParticles.Partition
      Float Selection Array=DataInterface() {propertyValues {FloatData [100,-100,40,-40]}}
     )
     Y=&Lerp_Float(
      A=0
      B=&FN_FrontEndFloatModifier(Gameplay=-280, Front End=-180)
      Alpha=$System.FN_Contrail_Setup.Dive Value
     )
     Z=0
    )) {module=AddVelocity001}
   &ShapeLocation(
    Cylinder Height=10
    Cylinder Height Midpoint=0
    Cylinder Radius=12
    MaskSpawnGroup=false
    Override Seed=false
    Shape Primitive=::NewEnumerator1
    Surface Only=true
    Surface Only Band Thickness=1
    Use Endcaps In Surface Only Mode=false
    UseOwnerScale=true
   )
  ]
  particleUpdateScript [
   &ParticleState()
   &SkeletalMeshLocation(
    Bone / Socket Index=$Particles.PartitionParticles.Partition
    Bone Sampling Mode=::NewEnumerator4
    Inherit Velocity=false
    Interpolate Mesh Sampling=false
    MaskSpawnGroup=false
    Skeletal Mesh=$System.Skeletal Mesh
   )
   &UpdateMeshOrientation(
    Rotation Rate=0.5
    Rotation Vector=&VectorFromFloat(Value=&RandomRangeFloat(Minimum=-0.5, Maximum=0.5))
   )
   &ScaleColor(
    Scale Alpha=&FloatFromCurve(
     FloatCurve=DataInterface() {propertyValues {Curve {keys [
         <at 0.05 0 RCIM_Cubic>
         {arriveTangentWeight=0.0306, interpMode=RCIM_Cubic, leaveTangentWeight=0.367, tangentMode=RCTM_User, time=0.1, value=1}
         {interpMode=RCIM_Cubic, leaveTangent=-2.6055, time=1, value=0}
        ]}}}
     Scale Curve=&FNTimeOfDayScalar(Night Scale=1, Morning Scale=1, Evening Scale=1)
    )
    Scale RGB=&MakeVectorFromLinearColorRGB(Color=&ColorFromCurve(ColorCurve=DataInterface() {propertyValues {
        AlphaCurve {keys []}
        BlueCurve {keys [<at 0 2 RCIM_Cubic>,<at 0.9338 1>]}
        GreenCurve {keys [<at 0 2 RCIM_Cubic>,<at 0.9338 0.1>]}
        RedCurve {keys [<at 0 0 RCIM_Cubic>,<at 0.9338 0.1079>]}
       }}))
    ScaleA=true
    ScaleRGB=true
   )
   &ScaleMeshSize(Scale Factor=&VectorFromFloat(Value=&FloatFromCurve(FloatCurve=DataInterface() {propertyValues {Curve {keys [
          {arriveTangent=4.3653, interpMode=RCIM_Cubic, leaveTangent=4.3653, tangentMode=RCTM_User, time=0, value=0.4}
          {arriveTangentWeight=0.0745, interpMode=RCIM_Cubic, leaveTangentWeight=0.2981, tangentMode=RCTM_User, time=0.35, value=1}
          {arriveTangent=-4.1118, interpMode=RCIM_Cubic, leaveTangent=-4.1118, tangentMode=RCTM_User, time=1, value=0}
         ]}}})))
   &Drag(Drag=4.8, Use Linear Drag=true, Use Rotational Drag=true)
   &SolveForcesAndVelocity(Clamp Velocity=false, Limit Acceleration=false)
  ]
  rendererClasses [&NiagaraMeshRendererProperties]
  renderers [{rendererClass=&NiagaraMeshRendererProperties, rendererIndex=0}]
 }
 Ribbon_Frosty_CubeMap {
  emitterUpdateScript [
   &EmitterState()
   /Game/Effects/Niagara/Modules/Skydiving/FN_DivingAttributes/FN_EmitterDivingAttributes_v1(Bone Index=1, SkeletalMesh:NiagaraDataInterfaceSkeletalMesh=$User.SkeletalMesh)
   &FN_SpawnRibbon_v2(Number of Ribbons=2, Spawn Mode=::NewEnumerator0, SpawnRate=70)
  ]
  particleSpawnScript [
   &FN_SetSpawnGroup(Number of Spawn Groups=2)
   &FN_InitializeRibbon_v3(
    Color:LinearColor=<rgba 0.25 0.25 0.25 1>
    Lifetime:NiagaraFloat=&Lerp_Float(A=0.3, B=0.4, Alpha=$System.FN_Contrail_Setup.Dive Value)
    Offset Axis:ENiagaraOrientationAxis=::ZAxis
    Ribbon Width=1
   )
   &SkeletalMeshLocation(
    Apply Sampled Orientation to Mesh Particles=true
    Bone / Socket Index=$Particles.SpawnGroup
    Bone Sampling Mode=::NewEnumerator4
    Inherit Velocity=false
    Interpolate Mesh Sampling=false
    MaskSpawnGroup=false
    Sampled Position Offset Coordinate Space:ENiagaraMeshTransforms=::NewEnumerator3
    Skeletal Mesh=$System.Skeletal Mesh
   )
   &AddVelocity(Velocity=&FN_FrontEndVector3Modifier(Gameplay=<0 0 550>, Front End=<0 0 550>))
   &AddVelocity(Velocity=&MakeVector(
     X=&SelectFloatFromArray(Direct Array Index=$Particles.SpawnGroup, Float Selection Array=DataInterface() {propertyValues {FloatData [100,-100]}})
     Y=&Lerp_Float(A=0, B=-150, Alpha=$System.FN_Contrail_Setup.Dive Value)
     Z=0
    )) {module=AddVelocity001}
  ]
  particleUpdateScript [
   &InheritSourceMovement(Applied Position Falloff Distance=100, Applied Position Scale=<0.9 0.9 1>, ApplyForce=false, ApplyPosition=true, ApplyPositionFalloff=false)
   &SolveForcesAndVelocity(Clamp Velocity=false, Limit Acceleration=false)
   &FN_UpdateRibbon(Detach Locked Particle=true, Enable Skeletal Position=true)
   &ScaleRibbonWidth(Ribbon Width Scale=&FloatFromCurve(
     FloatCurve=DataInterface() {propertyValues {Curve {keys [
         {arriveTangent=102.8897, interpMode=RCIM_Cubic, leaveTangent=102.8897, tangentMode=RCTM_User, time=0, value=0}
         {arriveTangent=57.0501, interpMode=RCIM_Cubic, leaveTangent=57.0501, tangentMode=RCTM_User, time=0.3, value=25}
         <at 1 45 RCIM_Cubic>
        ]}}}
     Scale Curve=&Lerp_Float(A=1, B=0.8, Alpha=$System.FN_Contrail_Setup.Dive Value)
    ))
   &ScaleColor(
    Scale Alpha=&FloatFromCurve(FloatCurve=DataInterface() {propertyValues {Curve {keys [
         {interpMode=RCIM_Cubic, leaveTangentWeight=0.5382, tangentMode=RCTM_User, time=0.08, value=0}
         {arriveTangentWeight=0.0622, interpMode=RCIM_Cubic, leaveTangentWeight=0.1555, tangentMode=RCTM_User, time=0.2, value=1}
         {arriveTangent=-2.7271, interpMode=RCIM_Cubic, leaveTangent=-2.7271, tangentMode=RCTM_User, time=0.5, value=0.5}
         <at 1 0 RCIM_Cubic>
        ]}}})
    Scale RGB=&MakeVectorFromLinearColorRGB(Color=&ColorFromCurve(ColorCurve=DataInterface() {propertyValues {
        AlphaCurve {keys []}
        BlueCurve {keys [<at 0.1 10>,<at 0.2 2>,<at 0.5 0.2>]}
        GreenCurve {keys [<at 0.1 9.6143>,<at 0.2 1.0818>,<at 0.5 0.034>]}
        RedCurve {keys [<at 0.1 6.3545>,<at 0.2 0.1988>,<at 0.5 0.0438>]}
       }}))
    ScaleA=true
    ScaleRGB=true
   )
   &ScaleColor(
    Scale Alpha=&FloatFromCurve(
     FloatCurve=DataInterface() {propertyValues {Curve {keys [
         <at 0 0.55 RCIM_Cubic>
         {arriveTangent=0.495, interpMode=RCIM_Cubic, leaveTangent=0.495, tangentMode=RCTM_User, time=0.25, value=0.6}
         <at 0.5 1 RCIM_Cubic>
         {arriveTangent=-0.4982, interpMode=RCIM_Cubic, leaveTangent=-0.4982, tangentMode=RCTM_User, time=0.75, value=0.6}
         <at 1 0.55 RCIM_Cubic>
        ]}}}
     CurveIndex=Linked() {linkedVariable {name=NPC.FortniteNPC.TimeOfDay_0-1, type=NiagaraFloat}}
     Scale Curve=&FN_FrontEndFloatModifier(Gameplay=1, Front End=0.7)
    )
    Scale RGB=&VectorFromFloat(Value=0.3)
    ScaleA=true
    ScaleRGB=true
   ) {module=ScaleColor001}
   &DynamicMaterialParameters(
    Index 0 Param 1=&FloatFromCurve(FloatCurve=DataInterface() {propertyValues {Curve {keys [<at 0 3>,<at 0.4 0>]}}}, Scale Curve=25)
    Index 0 Param 2=&FloatFromCurve(FloatCurve=DataInterface() {propertyValues {Curve {keys [
         {arriveTangent=12.7998, interpMode=RCIM_Cubic, leaveTangent=12.7998, tangentMode=RCTM_User, time=0, value=0.4}
         {arriveTangentWeight=0.0333, interpMode=RCIM_Cubic, leaveTangentWeight=0.1333, tangentMode=RCTM_User, time=0.1, value=1}
         {
          arriveTangent=-0.9247
          arriveTangentWeight=0.1602
          interpMode=RCIM_Cubic
          leaveTangent=-0.9247
          leaveTangentWeight=0.2003
          tangentMode=RCTM_User
          time=0.55
          value=0.8
         }
         {arriveTangent=-0.9567, interpMode=RCIM_Cubic, leaveTangent=-0.9567, tangentMode=RCTM_User, time=1, value=0.3}
        ]}}})
    Index 0 Param 3=-1
    Index 0 Param 4=&Lerp_Float(A=1, B=0.6, Alpha=$System.FN_Contrail_Setup.Dive Value)
    Module.Index0Param0Write=false
    Module.Index0Param1Write=true
    Module.Index0Param2Write=true
    Module.Index0Param3Write=true
   ) {module=DynamicMaterialParameters001}
  ]
  rendererClasses [&NiagaraRibbonRendererProperties]
  renderers [{rendererClass=&NiagaraRibbonRendererProperties}]
 }
 Ribbon_IceFog {
  emitterUpdateScript [
   &EmitterState()
   &FN_EmitterDivingAttributes_v1(Bone Index=1, SkeletalMesh:NiagaraDataInterfaceSkeletalMesh=$User.SkeletalMesh)
   &FN_SpawnRibbon_v2(
    Number of Ribbons:NiagaraInt32=&MakeIntFromFloat(
     Float=&ClampFloat(Float=&Lerp_Float(A=-50, B=400, Alpha=$System.FN_Contrail_Setup.Dive Value), Min=2, Max=2)
     Conversion Mode=::NewEnumerator2
    )
    Spawn Mode=::NewEnumerator0
    SpawnRate=70
   )
  ]
  particleSpawnScript [
   &FN_SetSpawnGroup()
   &FN_InitializeRibbon_v3(
    Color:LinearColor=&LerpLinearColors(StartColor=<rgba 0.1175 0.5588 1 0.7>, EndColor=<rgba 0 2 4 0.5>, LerpFactor=$System.FN_Contrail_Setup.Dive Value)
    Ribbon Width=1
   )
   &SkeletalMeshLocation(
    Apply Sampled Orientation to Mesh Particles=true
    Bone / Socket Index=$Particles.SpawnGroup
    Bone Sampling Mode=::NewEnumerator4
    Inherit Velocity=false
    Interpolate Mesh Sampling=false
    MaskSpawnGroup=false
    Sampled Position Offset Coordinate Space:ENiagaraMeshTransforms=::NewEnumerator3
    Skeletal Mesh=$System.Skeletal Mesh
   )
   &AddVelocity(Velocity=&FN_FrontEndVector3Modifier(Gameplay=<0 0 550>, Front End=<0 0 550>))
   &AddVelocity(Velocity=&MakeVector(
     X=&SelectFloatFromArray(Direct Array Index=$Particles.SpawnGroup, Float Selection Array=DataInterface() {propertyValues {FloatData [100,-100,30,-30]}})
     Y=&Lerp_Float(A=0, B=-150, Alpha=$System.FN_Contrail_Setup.Dive Value)
     Z=0
    )) {module=AddVelocity001}
   /SetVariables(Particles.Lifetime=&self:SpawnGrp_0_1_Mask(Masked Float:NiagaraFloat=0.2, Passthrough Float:NiagaraFloat=0.5)) {isSetParameters=true}
  ]
  particleUpdateScript [
   &InheritSourceMovement(Applied Position Falloff Distance=100, Applied Position Scale=<0.9 0.9 1>, ApplyForce=false, ApplyPosition=true, ApplyPositionFalloff=false)
   &SolveForcesAndVelocity(Clamp Velocity=false, Limit Acceleration=false)
   &FN_UpdateRibbon(Detach Locked Particle=true, Enable Skeletal Position=true)
   &ScaleRibbonWidth(Ribbon Width Scale=&FloatFromCurve(
     FloatCurve=DataInterface() {propertyValues {Curve {keys [
         {arriveTangent=455.1135, interpMode=RCIM_Cubic, leaveTangent=455.1135, tangentMode=RCTM_User, time=0, value=1}
         {arriveTangent=59, interpMode=RCIM_Cubic, leaveTangent=59, time=0.2, value=45}
         <at 1 60 RCIM_Cubic>
        ]}}}
     Scale Curve=&self:SpawnGrp_0_1_Mask(
      Masked Float:NiagaraFloat=0.2
      Passthrough Float:NiagaraFloat=&Lerp_Float(A=1, B=0.8, Alpha=$System.FN_Contrail_Setup.Dive Value)
     )
    )) {module=ScaleRibbonWidth001}
   &ScaleColor(
    Scale Alpha=&Multiply_Float(
     A=&FloatFromCurve(
      FloatCurve=DataInterface() {propertyValues {Curve {keys [
          {arriveTangent=8.011, interpMode=RCIM_Cubic, leaveTangent=8.011, tangentMode=RCTM_User, time=0, value=0.2}
          {arriveTangent=-0.2, arriveTangentWeight=0.1491, interpMode=RCIM_Cubic, leaveTangent=-0.2, leaveTangentWeight=0.2236, time=0.2, value=1}
          <at 1 0 RCIM_Cubic>
         ]}}}
      Scale Curve=&FN_FrontEndFloatModifier(Gameplay=1, Front End=0.45)
     )
     B=&self:SpawnGrp_0_1_Mask(
      Masked Float:NiagaraFloat=&ClampFloat(Float=&Lerp_Float(A=-1, B=1, Alpha=$System.FN_Contrail_Setup.Dive Value))
      Passthrough Float:NiagaraFloat=1
     )
    )
    Scale RGB=&MakeVectorFromLinearColorRGB(Color=&ColorFromCurve(ColorCurve=DataInterface() {propertyValues {
        AlphaCurve {keys []}
        BlueCurve {keys [<at 0 3.3877>,<at 0.25 2>,<at 0.6 1>]}
        GreenCurve {keys [<at 0 5>,<at 0.25 2>,<at 0.6 0.3546>]}
        RedCurve {keys [<at 0 2.7615>,<at 0.25 2>,<at 0.6 0>]}
       }}))
    ScaleA=true
    ScaleRGB=true
   )
   &ScaleColor(
    Scale Alpha=&FloatFromCurve(
     FloatCurve=DataInterface() {propertyValues {Curve {keys [
         <at 0 0.5 RCIM_Cubic>
         {arriveTangent=0.6396, interpMode=RCIM_Cubic, leaveTangent=0.6396, tangentMode=RCTM_User, time=0.25, value=0.55}
         {arriveTangentWeight=0.0833, interpMode=RCIM_Cubic, leaveTangentWeight=0.0833, tangentMode=RCTM_User, time=0.5, value=1}
         {arriveTangent=-0.6116, interpMode=RCIM_Cubic, leaveTangent=-0.6116, tangentMode=RCTM_User, time=0.75, value=0.55}
         <at 1 0.5 RCIM_Cubic>
        ]}}}
     CurveIndex=Linked() {linkedVariable {name=NPC.FortniteNPC.TimeOfDay_0-1, type=NiagaraFloat}}
    )
    ScaleA=true
    ScaleRGB=false
   ) {module=ScaleColor001}
  ]
  rendererClasses [&NiagaraRibbonRendererProperties]
  renderers [{rendererClass=&NiagaraRibbonRendererProperties}]
 }
}
summary {
 systemName=NS_Contrail_FrostedGlass
 emitters [
  {emitterName=Ribbon_Frosty_CubeMap, rendererClasses [&NiagaraRibbonRendererProperties]}
  {emitterName=Hand_IceFog, rendererClasses [&NiagaraSpriteRendererProperties]}
  {emitterName=Hand_Glow, rendererClasses [&NiagaraSpriteRendererProperties, &NiagaraLightRendererProperties]}
  {emitterName=Ribbon_IceFog, rendererClasses [&NiagaraRibbonRendererProperties]}
  {emitterName=Hand_SnowFlake, rendererClasses [&NiagaraMeshRendererProperties]}
 ]
 userVariables [
  User.bIsFrontEnd {type=NiagaraBool}
  User.bIsFrontEndPreview {type=NiagaraBool, defaultValue=true}
  User.ParaGlideLeanAlpha {type=NiagaraFloat, defaultValue=0, description=""}
  User.PlayerSpeed {type=NiagaraFloat, defaultValue=0, description=""}
  User.PoofSwitch {type=NiagaraFloat, defaultValue=0, description=""}
  User.RibbonIDOffset {type=NiagaraInt32}
  User.SkeletalMesh {
   type=NiagaraDataInterfaceSkeletalMesh
   defaultValue {
    struct=NiagaraExt_VariableValue_DataInterface
    value {dataInterface=&self:NiagaraDataInterfaceSkeletalMesh_1}
   }
  }
 ]
}
